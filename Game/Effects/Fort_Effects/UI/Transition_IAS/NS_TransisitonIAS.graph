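# Niagara UI transition effect (Transition_IAS) that draws an animated, status-colored torus of mesh/sprite cubes and three ribbon layers.
# Curl noise and traveling sine waves animate the ring, while User.ResInProgress fades ribbons and drives cube snapping to skeletal-mesh samples.
# After System.EndTime, cubes release staggered by torus position, pulled axially and launched upward, as ribbons animate out.
# Camera-based ConstantScreenSize modules in every emitter use User.ScreenSize as reference width.
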NiagaraSystem /Game/Effects/Fort_Effects/UI/Transition_IAS/NS_TransisitonIAS
dependencies {
 usedDataInterfaces [
  &NiagaraDataInterfaceArrayFloat
  &NiagaraDataInterfaceCurve
  &NiagaraDataInterfaceCurlNoise
  &NiagaraDataInterfaceCamera
  &NiagaraDataInterfaceColorCurve
  &NiagaraDataInterfaceArrayColor
 ]
 usedDynamicInputs [
  &SelectFloatFromArray
  &SelectLinearColorFromArray
  &HueShiftLinearColor
 ]
 usedModules [
  &self:SystemSpawnScript.SystemScriptSource.SystemScriptGraph.NiagaraNodeAssignment_1.SetVariables_#1
  &SystemState
  &DoOnce
  &self:SystemSpawnScript.SystemScriptSource.SystemScriptGraph.NiagaraNodeAssignment_2.SetVariables_#2
  &self:SystemSpawnScript.SystemScriptSource.SystemScriptGraph.NiagaraNodeAssignment_3.SetVariables_#3
  &self:SystemSpawnScript.SystemScriptSource.SystemScriptGraph.NiagaraNodeAssignment_4.SetVariables_#4
  &self:SystemSpawnScript.SystemScriptSource.SystemScriptGraph.NiagaraNodeAssignment_0.SetVariables_#5
 ]
}
emitters {
 Cubes {
  emitterUpdateScript [
   &EmitterState()
   &SpawnRate(SpawnRate=100, Use Spawn Probability=false) {enabled=false}
   &SpawnBurst_Instantaneous(Spawn Count=140, Use Spawn Probability=false, UseLoopCountLimit=false)
  ]
  particleSpawnScript [
   &InitializeParticle(
    Color=$System.Status Color
    Mesh Scale Mode=::NewEnumerator3
    Sprite Size Mode=::NewEnumerator3
    Uniform Sprite Size=15
    UsePositionOffset=false
    Write Material Random=true
   )
   &ShapeLocation(
    MaskSpawnGroup=false
    Ring / Disc Distribution Mode:ENiagaraSphereDistributionMode=::NewEnumerator3
    Ring Radius=25
    Shape Primitive=::NewEnumerator4
    Uniform Spiral Falloff=25
    UseOwnerScale=true
   ) {module=ShapeLocation001}
   /SetVariables(Particles.Radius=60) {isSetParameters=true}
   /SetVariables(Particles.Thickness=20) {isSetParameters=true}
   &ShapeLocation(
    Handle Radius=0
    Large Radius=$Particles.Radius
    MaskSpawnGroup=false
    Shape Primitive=::NewEnumerator3
    Torus Distribution Mode:ENiagaraTorusDistributionMode=::NewEnumerator1
    U Position=&ReturnNormalizedExecIndex()
    UseOwnerScale=true
    V Position=&RandomRangeFloat(Minimum=0, Maximum=1)
   )
   &ShapeLocation(MaskSpawnGroup=false, Override Seed=false, Sphere Radius=$Particles.Thickness, Sphere Surface Distribution=1, UseOwnerScale=true) {module=ShapeLocation002}
   &self:MeshSize(
    Curve for Floats=DataInterface() {propertyValues {Curve {keys [
        {interpMode=RCIM_Cubic, leaveTangent=1.0974, tangentMode=RCTM_Break, time=-1, value=0.6}
        {arriveTangent=1.0521, interpMode=RCIM_Cubic, tangentMode=RCTM_Break, time=1, value=1.7}
       ]}}}
    Curve for Floats001=DataInterface() {propertyValues {Curve {keys [{interpMode=RCIM_Cubic, tangentMode=RCTM_Break, time=-1, value=1}, {interpMode=RCIM_Cubic, tangentMode=RCTM_Break, time=1, value=1.5}]}}}
    HeightArea=20
   ) {module=MeshSize}
   &InitialMeshOrientation(Use Rotation Vector=false)
   &self:TorusPos(Position:NiagaraPosition=$Particles.Position) {module=TorusPos001}
   /SetVariables(Particles.OriginalPosition:NiagaraPosition=$Particles.Position) {isSetParameters=true}
   &self:CustomSample(
    R=40
    Skeletal Mesh:NiagaraDataInterfaceSkeletalMesh=$User.SKM
    TorusProgress:NiagaraFloat=&Subtract_Float(A=1, B=$Particles.TorusProgress)
   ) {module=CustomSample}
  ]
  particleUpdateScript [
   /SetVariables(Particles.Color=$System.Status Color) {enabled=false, isSetParameters=true}
   /SetVariables(Particles.OutTime:NiagaraFloat=$System.EndTime) {isSetParameters=true}
   &ParticleState(Kill Particles When Lifetime Has Elapsed=false, Let Infinitely Lived Particles Die When Emitter Deactivates=true)
   /SetVariables() {isSetParameters=true}
   /SetVariables(Particles.UnCWeight:NiagaraFloat=&Subtract_Float(
     A=&Add_Float(A=$Emitter.Age, B=&Multiply_Float(A=$Particles.Initial.TorusProgress, B=-0.2))
     B=$Particles.OutTime
    )) {isSetParameters=true}
   /SetVariables(Particles.Weight:NiagaraFloat=&ClampFloat(Float=$Particles.UnCWeight)) {isSetParameters=true}
   /SetVariables(Particles.Position=&Lerp_Vector(
     A=Linked() {linkedVariable {name=Particles.OriginalPosition, type=NiagaraPosition}}
     B=Linked() {linkedVariable {name=Particles.Position, type=NiagaraPosition}}
     Alpha=&MakeCustomFloatFromBool(Bool=&SetBoolByFloatComparison(A=$Particles.Weight, B=0), True Float=1, False Float=0)
    )) {isSetParameters=true}
   &self:--------------------------() {module=--------------------------}
   &self:CircleVelocity(AC=-100, RotateAxis:Vector3f=<0 0 1>, RotateSpeed=0.02) {module=CircleVelocity}
   &self:Rotation(
    Position:NiagaraPosition=$Particles.CirclePosition
    R=0.03
    Weight:NiagaraFloat=&Power(a=$Particles.Weight, b=0.2)
    float:NiagaraFloat=&Add_Float(A=$Particles.Radius, B=0)
    rds:NiagaraFloat=$Particles.Radius
   ) {module=Rotation}
   &self:--------------------------() {module=--------------------------001}
   /SetVariables(Particles.Position:Vector3f=$Particles.RPos) {isSetParameters=true}
   /SetVariables(Particles.Velocity=&AddVectorToPosition(
     Position=Linked() {linkedVariable {name=Particles.Velocity, type=Vector3f}}
     Vector=&Lerp_Vector(
      A=<0 0 0>
      B=&Multiply_VectorByFloat(Vector=<0 0 30000>, Float=&RandomRangeFloat(Minimum=1, Maximum=1))
      Alpha=&FloatFromCurve(
       FloatCurve=DataInterface() {propertyValues {Curve {keys [
           {interpMode=RCIM_Cubic, leaveTangent=0.0956, tangentMode=RCTM_Break, time=0, value=0}
           {arriveTangent=2.1575, interpMode=RCIM_Cubic, tangentMode=RCTM_Break, time=0.8, value=1}
           <at 0.9 0>
          ]}}}
       CurveIndex=$Particles.UnCWeight
      )
     )
    )) {isSetParameters=true}
   &VortexForce(
    Origin Pull Amount:NiagaraFloat=&Lerp_Float(
     A=0
     B=30000
     Alpha=&FloatFromCurve(
      FloatCurve=DataInterface() {propertyValues {Curve {keys [{interpMode=RCIM_Cubic, leaveTangent=8.9221, tangentMode=RCTM_Break, time=0, value=0}, <at 0.2 1>]}}}
      CurveIndex=$Particles.UnCWeight
     )
    )
    UseOriginOffset=false
    Vortex Force Amount=&Lerp_Float(
     A=0
     B=0
     Alpha=&FloatFromCurve(
      FloatCurve=DataInterface() {propertyValues {Curve {keys [{interpMode=RCIM_Cubic, leaveTangent=8.9221, tangentMode=RCTM_Break, time=0, value=0}, <at 0.2 1>]}}}
      CurveIndex=$Particles.UnCWeight
     )
    )
   )
   /Niagara/Modules/Update/Forces/LineAttractionForce(
    Attraction Strength:NiagaraFloat=&Lerp_Float(
     A=0
     B=30000
     Alpha=&FloatFromCurve(
      FloatCurve=DataInterface() {propertyValues {Curve {keys [
          {interpMode=RCIM_Cubic, tangentMode=RCTM_Break, time=0, value=0}
          {interpMode=RCIM_Cubic, tangentMode=RCTM_Break, time=0.1, value=1}
          <at 0.2 0>
         ]}}}
      CurveIndex=$Particles.UnCWeight
     )
    )
    Line End:Vector3f=<0 0 1000>
    Line Start:Vector3f=<0 0 0>
    Use Attraction Falloff=false
   )
   /SetVariables(Transient.PhysicsForce:Vector3f=&Multiply_VectorByFloat(
     Vector=$Transient.PhysicsForce
     Float=&ClampFloat(Float=&Divide_Float(A=0.016, B=$Engine.DeltaTime), Max=3)
    )) {isSetParameters=true}
   &self:--------------------------() {module=--------------------------004}
   &UpdateMeshOrientation(
    Rotation Rate=0.5
    Rotation Vector=&NormalizeVector(Vector To Normalize=$Particles.FVelocity, Safe Normalize=true, Safe Fallback Threshold=0, Safe Fallback Vector=<0 0 0.001>)
   )
   &SolveForcesAndVelocity(Clamp Velocity=false, Limit Acceleration=false)
   /SetVariables(Particles.OriginalPosition:NiagaraPosition=$Particles.Position) {isSetParameters=true}
   &self:--------------------------() {module=--------------------------002}
   &self:Constraint(
    Position:NiagaraPosition=$Particles.Position
    R:NiagaraFloat=$Particles.Thickness
    Weight:NiagaraFloat=&MakeCustomFloatFromBool(Bool=&SetBoolByFloatComparison(A=$Particles.Weight, B=0), True Float=1, False Float=0)
    float:NiagaraFloat=$Particles.Radius
   ) {module=Constraint}
   &self:TorusPos(Position:NiagaraPosition=$Particles.Position) {module=TorusPos}
   &self:ScaleWave(
    Curve for Floats=DataInterface() {propertyValues {Curve {keys [{interpMode=RCIM_Cubic, leaveTangent=1.3961, tangentMode=RCTM_Break, time=0, value=0.5}, <at 1 1>]}}}
    Radius:NiagaraFloat=$Particles.Radius
    Scale:Vector3f=<1.3 1.3 1.3>
    Weight:NiagaraFloat=&MakeCustomFloatFromBool(Bool=&SetBoolByFloatComparison(A=$Particles.Weight, B=0), True Float=1, False Float=0)
    float:NiagaraFloat=&FloatFromCurve(
     FloatCurve=DataInterface() {propertyValues {Curve {keys [<at -1 0.2>, {arriveTangent=1.461, time=1, value=1}]}}}
     CurveIndex=&Sine(
      Normalized Angle=&Add_Float(A=$Particles.TorusProgress, B=&Multiply_Float(A=$Emitter.Age, B=-0.1875))
      Period=0.5
      Scale=1
      Bias=0
     )
     Scale Curve=0.6
    )
   ) {module=ScaleWave}
   &self:NoiseWave(
    Curl Noise=DataInterface() {propertyValues {Seed=32}}
    Quat:Quat4f=&MakeQuatFromAxisAngle(Axis=<0 0 1>, AngleInDegrees=&Multiply_Float(A=$Emitter.Age, B=75))
    Vector:Vector3f=<0 0 2>
    Weight:NiagaraFloat=&MakeCustomFloatFromBool(Bool=&SetBoolByFloatComparison(A=$Particles.Weight, B=0), True Float=1, False Float=0)
    float=7
    float001=0.002
   ) {module=NoiseWave}
   &self:--------------------------() {module=--------------------------003}
   /SetVariables(Particles.SVelocity:Vector3f=&Lerp_Vector(
     A=$Particles.FVelocity
     B=&Multiply_VectorByFloat(Vector=$Particles.Velocity, Float=10)
     Alpha=&FloatFromCurve(FloatCurve=DataInterface() {propertyValues {Curve {keys [<at 0 0>,<at 0.1 1>]}}}, CurveIndex=$Particles.Weight)
    )) {isSetParameters=true}
   &ScaleMeshSize(Scale Factor=&VectorFromFloat(Value=&FloatFromCurve(
      FloatCurve=DataInterface() {propertyValues {Curve {keys [<at 0 0>,<at 0.2 1>]}}}
      CurveIndex=$Particles.Age
      Scale Curve=&FloatFromCurve(FloatCurve=DataInterface() {propertyValues {Curve {keys [<at 0 1>,<at 0.4 0>]}}}, CurveIndex=$Particles.UnCWeight)
     )))
   &ScaleMeshSize(Scale Factor=&VectorFromFloat(Value=&FloatFromCurve(
      FloatCurve=DataInterface() {propertyValues {Curve {keys [<at 0 0>,<at 0.2 1>]}}}
      CurveIndex=$Particles.Age
      Scale Curve=&FloatFromCurve(
       FloatCurve=DataInterface() {propertyValues {Curve {keys [<at 0 1>,<at 0.4 0>]}}}
       CurveIndex=&Add_Float(
        A=&Subtract_Float(A=1, B=$Particles.Initial.TorusProgress)
        B=&Subtract_Float(A=&Add_Float(A=$Emitter.Age, B=-0.5), B=$Particles.OutTime)
       )
      )
     ))) {enabled=false, module=ScaleMeshSize001}
   &ScaleSpriteSize(
    Uniform Curve Index=$Particles.Age
    Uniform Curve Sprite Scale=DataInterface() {propertyValues {Curve {keys [{tangentMode=RCTM_User, time=0, value=0}, {tangentMode=RCTM_User, time=0.05, value=1}]}}}
   )
   &ScaleColor(
    Color Value To Scale=$System.Status Color
    Scale Alpha=&Multiply_Float(
     A=&FloatFromCurve(
      FloatCurve=DataInterface() {propertyValues {Curve {keys [{interpMode=RCIM_Cubic, tangentMode=RCTM_Break, time=0, value=0}, <at 1 1>]}}}
      CurveIndex=$Particles.Age
      Scale Curve=&FloatFromCurve(
       FloatCurve=DataInterface() {propertyValues {Curve {keys [
           {interpMode=RCIM_Cubic, leaveTangent=-7.0896, tangentMode=RCTM_Break, time=0, value=1}
           {arriveTangent=-1.1691, tangentMode=RCTM_User, time=0.15, value=0}
          ]}}}
       CurveIndex=$Particles.UnCWeight
      )
     )
     B=1
    )
    ScaleA=true
    ScaleRGB=true
   )
   &self:Into(
    Curl Noise=DataInterface() {propertyValues {Seed=2}}
    Curve for Floats=DataInterface() {propertyValues {Curve {keys [<at 0 1>,<at 1 0>]}}}
    Outer Radius:NiagaraFloat=&FloatFromCurve(
     FloatCurve=DataInterface() {propertyValues {Curve {keys [<at 0 0.6>,<at 0.2 1>]}}}
     CurveIndex=&Add_Float(
      A=&Add_Float(A=$Particles.Age, B=-0.7)
      B=&Subtract_Float(A=1, B=$Particles.Initial.TorusProgress)
     )
     Scale Curve=10
    )
    Position:NiagaraPosition=$StackContext.Position
    Quaternion:Quat4f=&MakeQuatFromAxisAngle(Axis=<0 0 1>, AngleInDegrees=&Multiply_Float(A=$Emitter.Age, B=50))
    TorusPos:NiagaraFloat=$Particles.Initial.TorusProgress
    Vector001:Vector3f=<180 0 -80>
    float=0.5
    float001:NiagaraFloat=&Multiply_Float(A=&Multiply_Float(A=&Subtract_Float(A=$Particles.Age, B=0.5), B=2), B=-0.5)
    float002=0.01
    float003=7
   ) {module=Into}
   /SetVariables(
    Particles.Scale2:Vector3f=&Multiply_Vector(A=$Particles.Scale2, B=$Particles.RRScale)
    Particles.SpriteSize2=&Multiply_Vector2D_ByFloat(Float=1, Vector2D=$Particles.SpriteSize2)
   ) {isSetParameters=true}
   &KillParticles(Kill Particles=&SetBoolByFloatComparison(A=$Particles.UnCWeight, B=1.2))
   &self:SnapToSkeletalmesh(
    M=2
    Progress:NiagaraFloat=$User.ResInProgress
    SnappedPosition:NiagaraPosition=$Particles.CustomSample.SampledPosition
    TorusProgress:NiagaraFloat=&Subtract_Float(A=1, B=$Particles.Initial.TorusProgress)
   ) {module=SnapToSkeletalmesh}
   &ScaleColor(
    Color Value To Scale=$System.Status Color
    Scale Alpha=&FloatFromCurve(FloatCurve=DataInterface() {propertyValues {Curve {keys [<at 1 1>,<at 1.3 0>]}}}, CurveIndex=$Particles.OriginProgress)
    ScaleA=true
    ScaleRGB=true
   ) {module=ScaleColor001}
   /SetVariables(Particles.ColorC:LinearColor=&MultiplyLinearColor(
     Linear Color=&MultiplyLinearColor(
      Linear Color=$Particles.Color
      Scale Factor=&MakeVector4(
       X=1
       Y=1
       Z=1
       W=&MakeFloatFromLinearColor(LinearColor=$Particles.SColor, Channel:ENiagaraLinearColor_Channels=::NewEnumerator3)
      )
     )
     Scale Factor=<1 1 1 0.02>
    )) {isSetParameters=true}
   &self:ConstantScreenSize(
    Camera Query=DataInterface() {propertyValues {PlayerControllerIndex=0, bRequireCurrentFrameData=false}}
    Position:NiagaraPosition=$Particles.LerpPosition
    Reference Screen Width:NiagaraFloat=$User.ScreenSize
    Scale:Vector3f=$Particles.Scale2
    SpriteSize:Vector2f=$Particles.SpriteSize2
   ) {module=ConstantScreenSize}
  ]
  rendererClasses [&NiagaraMeshRendererProperties, &NiagaraSpriteRendererProperties]
  renderers [
   {rendererClass=&NiagaraMeshRendererProperties, rendererIndex=0}
   {rendererClass=&NiagaraSpriteRendererProperties, rendererIndex=1}
  ]
 }
 DownRibbon {
  emitterUpdateScript [
   &EmitterState()
   &SpawnBurst_Instantaneous(Spawn Count=20, Use Spawn Probability=false, UseLoopCountLimit=false)
   &SpawnBurst_Instantaneous(Spawn Count=20, Spawn Time=6.3167, Use Spawn Probability=false, UseLoopCountLimit=false) {module=SpawnBurst_Instantaneous001}
   &SpawnBurst_Instantaneous(Spawn Count=20, Spawn Time=2.3333, Use Spawn Probability=false, UseLoopCountLimit=false) {module=SpawnBurst_Instantaneous002}
  ]
  particleSpawnScript [
   &InitializeParticle(
    Color=&MultiplyLinearColor(Linear Color=$System.Status Color, Scale Factor=<1 1 1 0.05>)
    Lifetime=5
    Ribbon Width=40
    Ribbon Width Mode=::NewEnumerator1
    Sprite Size Mode=::NewEnumerator3
    UsePositionOffset=false
    Write Material Random=true
   )
   /SetVariables(Particles.Radius=60) {isSetParameters=true, module=SetVariables_#6}
   /SetVariables(Particles.Thickness=20) {isSetParameters=true, module=SetVariables_#7}
   &ShapeLocation(
    Handle Radius=0
    Large Radius=$Particles.Radius
    MaskSpawnGroup=false
    Offset=<0 0 -15>
    Offset Mode=::NewEnumerator0
    Shape Primitive=::NewEnumerator3
    Torus Distribution Mode:ENiagaraTorusDistributionMode=::NewEnumerator1
    U Position=&ReturnNormalizedExecIndex()
    UseOwnerScale=true
    V Position=&RandomRangeFloat(Minimum=0, Maximum=1)
   )
   /SetVariables(Particles.RibbonFacing=<0 0 -1>) {isSetParameters=true}
  ]
  particleUpdateScript [
   &ParticleState(Kill Particles When Lifetime Has Elapsed=false, Let Infinitely Lived Particles Die When Emitter Deactivates=true)
   /SetVariables(Particles.Position=$Particles.Initial.Position) {isSetParameters=true, module=SetVariables_#8}
   &self:TorusPos(Position:NiagaraPosition=$Particles.Position) {module=TorusPos}
   &self:NoiseWave(
    Curl Noise=DataInterface() {propertyValues {Seed=32}}
    Quat:Quat4f=&MakeQuatFromAxisAngle(Axis=<0 0 1>, AngleInDegrees=&Multiply_Float(A=$Emitter.Age, B=75))
    Vector:Vector3f=<0 0 2>
    float=7
    float001=0.002
   ) {module=NoiseWave}
   /SetVariables(Particles.ScaleRibbonF:NiagaraFloat=&FloatFromCurve(
     FloatCurve=DataInterface() {propertyValues {Curve {keys [<at -1 0>, {arriveTangent=1.461, time=1, value=1}]}}}
     CurveIndex=&Sine(
      Normalized Angle=&Add_Float(A=$Particles.TorusProgress, B=&Multiply_Float(A=$Emitter.Age, B=-0.25))
      Period=0.5
      Scale=1
      Bias=0
     )
     Scale Curve=0.6
    )) {isSetParameters=true, module=SetVariables_#9}
   &ScaleRibbonWidth(Ribbon Width Scale=&FloatFromCurve(
     FloatCurve=DataInterface() {propertyValues {Curve {keys [
         {interpMode=RCIM_Cubic, tangentMode=RCTM_Break, time=0, value=1}
         {arriveTangent=0.1603, interpMode=RCIM_Cubic, tangentMode=RCTM_Break, time=1, value=1}
        ]}}}
     CurveIndex=$Particles.ScaleRibbonF
    ))
   &ScaleColor(
    Color Value To Scale=$System.Status Color
    Curve Index=$Particles.ScaleRibbonF
    Linear Color Curve=DataInterface() {propertyValues {AlphaCurve {keys [<at 0 1>,<at 1 1>]}, BlueCurve {keys [<at 0.0659 2>]}, GreenCurve {keys [<at 0.0659 2>]}, RedCurve {keys [<at 0.0659 2>]}}}
    Scale Mode=::NewEnumerator2
   )
   &ScaleColor(
    Color Value To Scale=$System.Status Color
    Scale Alpha=&FloatFromCurve(
     FloatCurve=DataInterface() {propertyValues {Curve {keys [{interpMode=RCIM_Cubic, tangentMode=RCTM_Break, time=0.1, value=0}, {interpMode=RCIM_Cubic, tangentMode=RCTM_Break, time=2, value=1}]}}}
     CurveIndex=$Particles.Age
     Scale Curve=&FloatFromCurve(
      FloatCurve=DataInterface() {propertyValues {Curve {keys [{interpMode=RCIM_Cubic, leaveTangent=-4.2363, tangentMode=RCTM_Break, time=0, value=1}, <at 0.5 0>]}}}
      CurveIndex=&Subtract_Float(A=$Emitter.Age, B=$System.EndTime)
     )
    )
    ScaleA=true
    ScaleRGB=true
   ) {module=ScaleColor001}
   &ScaleColor(
    Color Value To Scale=$System.Status Color
    Scale Alpha=&FloatFromCurve(FloatCurve=DataInterface() {propertyValues {Curve {keys [<at 0.5 0>,<at 1 1>]}}}, CurveIndex=$User.ResInProgress)
    ScaleA=true
    ScaleRGB=true
   ) {module=ScaleColor004}
   &self:ConstantScreenSize(
    Camera Query=DataInterface() {propertyValues {PlayerControllerIndex=0, bRequireCurrentFrameData=false}}
    Position:NiagaraPosition=$Particles.Position
    Reference Screen Width:NiagaraFloat=$User.ScreenSize
    RibbonWidth:NiagaraFloat=$Particles.RibbonWidth
   ) {module=ConstantScreenSize}
  ]
  rendererClasses [&NiagaraRibbonRendererProperties, &NiagaraSpriteRendererProperties]
  renderers [
   {rendererClass=&NiagaraRibbonRendererProperties}
   {rendererClass=&NiagaraSpriteRendererProperties, rendererIndex=1}
  ]
 }
 PianoRibbon {
  emitterUpdateScript [
   &EmitterState()
   &SpawnBurst_Instantaneous(Spawn Count=20, Use Spawn Probability=false, UseLoopCountLimit=false)
  ]
  particleSpawnScript [
   &InitializeParticle(
    Color=&MultiplyLinearColor(Linear Color=$System.Status Color, Scale Factor=<1 1 1 0.05>)
    Lifetime=5
    Ribbon Width=10
    Ribbon Width Mode=::NewEnumerator1
    Sprite Size Mode=::NewEnumerator3
    UsePositionOffset=false
    Write Material Random=true
   )
   /SetVariables(Particles.Radius=60) {isSetParameters=true, module=SetVariables_#6}
   /SetVariables(Particles.Thickness=20) {isSetParameters=true, module=SetVariables_#7}
   &ShapeLocation(
    Handle Radius=0
    Large Radius=$Particles.Radius
    MaskSpawnGroup=false
    Shape Primitive=::NewEnumerator3
    Torus Distribution Mode:ENiagaraTorusDistributionMode=::NewEnumerator1
    U Position=&ReturnNormalizedExecIndex()
    UseOwnerScale=true
    V Position=&RandomRangeFloat(Minimum=0, Maximum=1)
   )
  ]
  particleUpdateScript [
   /SetVariables(Particles.Color=$System.Status Color) {isSetParameters=true}
   &ParticleState(Kill Particles When Lifetime Has Elapsed=false, Let Infinitely Lived Particles Die When Emitter Deactivates=true)
   /SetVariables(Particles.Position=$Particles.Initial.Position) {isSetParameters=true, module=SetVariables_#8}
   &self:TorusPos(Position:NiagaraPosition=$Particles.Position) {module=TorusPos}
   &self:NoiseWave(
    Curl Noise=DataInterface() {propertyValues {Seed=32}}
    Quat:Quat4f=&MakeQuatFromAxisAngle(Axis=<0 0 1>, AngleInDegrees=&Multiply_Float(A=$Emitter.Age, B=75))
    Vector:Vector3f=<0 0 2>
    float=7
    float001=0.002
   ) {module=NoiseWave}
   /SetVariables(Particles.ScaleRibbonF:NiagaraFloat=&FloatFromCurve(
     FloatCurve=DataInterface() {propertyValues {Curve {keys [<at -1 0>, {arriveTangent=1.461, time=1, value=1}]}}}
     CurveIndex=&Sine(
      Normalized Angle=&Add_Float(A=$Particles.TorusProgress, B=&Multiply_Float(A=$Emitter.Age, B=-0.25))
      Period=0.5
      Scale=1
      Bias=0
     )
     Scale Curve=0.6
    )) {isSetParameters=true, module=SetVariables_#9}
   &ScaleRibbonWidth(Ribbon Width Scale=&FloatFromCurve(
     FloatCurve=DataInterface() {propertyValues {Curve {keys [
         {interpMode=RCIM_Cubic, tangentMode=RCTM_Break, time=0, value=0.15}
         {arriveTangent=0.9029, interpMode=RCIM_Cubic, tangentMode=RCTM_Break, time=1, value=1}
        ]}}}
     CurveIndex=$Particles.ScaleRibbonF
    ))
   &ScaleColor(
    Color Value To Scale=$System.Status Color
    Curve Index=$Particles.ScaleRibbonF
    Linear Color Curve=DataInterface() {propertyValues {
      AlphaCurve {keys [<at 0 1>,<at 1 1>]}
      BlueCurve {keys [<at 0.0659 10>]}
      GreenCurve {keys [<at 0.0659 10>]}
      RedCurve {keys [<at 0.0659 10>]}
     }}
    Scale Mode=::NewEnumerator2
   )
   &ScaleColor(
    Color Value To Scale=$System.Status Color
    Scale Alpha=&FloatFromCurve(
     FloatCurve=DataInterface() {propertyValues {Curve {keys [{interpMode=RCIM_Cubic, tangentMode=RCTM_Break, time=0.5, value=0}, {interpMode=RCIM_Cubic, tangentMode=RCTM_Break, time=2, value=1}]}}}
     CurveIndex=$Particles.Age
    )
    ScaleA=true
    ScaleRGB=true
   ) {module=ScaleColor001}
   &DynamicMaterialParameters(
    Index 0 Param 1=&FloatFromCurve(
     FloatCurve=DataInterface() {propertyValues {Curve {keys [{interpMode=RCIM_Cubic, leaveTangent=-4.1783, tangentMode=RCTM_Break, time=0, value=0}, <at 0.3 -1>]}}}
     CurveIndex=&Subtract_Float(A=$Emitter.Age, B=$System.EndTime)
    )
    Module.Index0Param0Write=true
    Module.Index0Param1Write=true
    Module.Index0Param2Write=true
    Module.Index0Param3Write=true
   )
   &ScaleColor(
    Color Value To Scale=$System.Status Color
    Scale Alpha=&FloatFromCurve(FloatCurve=DataInterface() {propertyValues {Curve {keys [<at 0.5 0>,<at 1 1>]}}}, CurveIndex=$User.ResInProgress)
    ScaleA=true
    ScaleRGB=true
   ) {module=ScaleColor004}
   &self:ConstantScreenSize(
    Camera Query=DataInterface() {propertyValues {PlayerControllerIndex=0, bRequireCurrentFrameData=false}}
    Position:NiagaraPosition=$Particles.Position
    Reference Screen Width:NiagaraFloat=$User.ScreenSize
    RibbonWidth:NiagaraFloat=$Particles.RibbonWidth
   ) {module=ConstantScreenSize}
  ]
  rendererClasses [&NiagaraRibbonRendererProperties]
  renderers [{rendererClass=&NiagaraRibbonRendererProperties}]
 }
 SoftRibbon1 {
  emitterUpdateScript [
   &EmitterState()
   &SpawnBurst_Instantaneous(Spawn Count=20, Use Spawn Probability=false, UseLoopCountLimit=false)
  ]
  particleSpawnScript [
   &InitializeParticle(
    Color=&MultiplyLinearColor(Linear Color=$System.Status Color, Scale Factor=<1 1 1 0.05>)
    Lifetime=5
    Ribbon Width=50
    Ribbon Width Mode=::NewEnumerator1
    Sprite Size Mode=::NewEnumerator3
    UsePositionOffset=false
    Write Material Random=true
   )
   /SetVariables(Particles.Radius=70) {isSetParameters=true, module=SetVariables_#6}
   /SetVariables(Particles.Thickness=20) {isSetParameters=true, module=SetVariables_#7}
   &ShapeLocation(
    Handle Radius=0
    Large Radius=$Particles.Radius
    MaskSpawnGroup=false
    Shape Primitive=::NewEnumerator3
    Torus Distribution Mode:ENiagaraTorusDistributionMode=::NewEnumerator1
    U Position=&ReturnNormalizedExecIndex()
    UseOwnerScale=true
    V Position=&RandomRangeFloat(Minimum=0, Maximum=1)
   )
   &self:TorusPos(Position:NiagaraPosition=$Particles.Position) {module=TorusPos001}
  ]
  particleUpdateScript [
   /SetVariables(Particles.Color=$System.Status Color) {enabled=false, isSetParameters=true}
   /SetVariables(Particles.OutTime:NiagaraFloat=$System.EndTime) {isSetParameters=true}
   &ParticleState(Kill Particles When Lifetime Has Elapsed=false, Let Infinitely Lived Particles Die When Emitter Deactivates=true)
   /SetVariables() {isSetParameters=true}
   /SetVariables(Particles.UnCWeight:NiagaraFloat=&Subtract_Float(
     A=&Add_Float(A=$Emitter.Age, B=&Multiply_Float(A=$Particles.Initial.TorusProgress, B=-0.5))
     B=$Particles.OutTime
    )) {isSetParameters=true}
   /SetVariables(Particles.Weight:NiagaraFloat=&ClampFloat(Float=$Particles.UnCWeight)) {isSetParameters=true}
   /SetVariables(Particles.Position=&Lerp_Vector(
     A=Linked() {linkedVariable {name=Particles.Initial.Position, type=NiagaraPosition}}
     B=Linked() {linkedVariable {name=Particles.Position, type=NiagaraPosition}}
     Alpha=&MakeCustomFloatFromBool(Bool=&SetBoolByFloatComparison(A=$Particles.Weight, B=0), True Float=1, False Float=0)
    )) {isSetParameters=true}
   &self:TorusPos(Position:NiagaraPosition=$Particles.Position) {module=TorusPos}
   &self:NoiseWave(
    Curl Noise=DataInterface() {propertyValues {Seed=32}}
    Quat:Quat4f=&MakeQuatFromAxisAngle(Axis=<0 0 1>, AngleInDegrees=&Multiply_Float(A=$Emitter.Age, B=75))
    Vector:Vector3f=<0 0 2>
    Weight:NiagaraFloat=&MakeCustomFloatFromBool(Bool=&SetBoolByFloatComparison(A=$Particles.Weight, B=0), True Float=1, False Float=0)
    float=7
    float001=0.002
   ) {module=NoiseWave}
   /SetVariables(Particles.ScaleRibbonF:NiagaraFloat=&FloatFromCurve(
     FloatCurve=DataInterface() {propertyValues {Curve {keys [<at -1 0>, {arriveTangent=1.461, time=1, value=1}]}}}
     CurveIndex=&Sine(
      Normalized Angle=&Add_Float(A=$Particles.TorusProgress, B=&Multiply_Float(A=$Emitter.Age, B=-0.25))
      Period=0.5
      Scale=1
      Bias=0
     )
     Scale Curve=0.6
    )) {isSetParameters=true, module=SetVariables_#9}
   &ScaleRibbonWidth(Ribbon Width Scale=&Lerp_Float(
     A=&FloatFromCurve(
      FloatCurve=DataInterface() {propertyValues {Curve {keys [{interpMode=RCIM_Cubic, tangentMode=RCTM_Break, time=0, value=0.3}, <at 1 1>]}}}
      CurveIndex=$Particles.ScaleRibbonF
     )
     B=1.1
     Alpha=&FloatFromCurve(
      FloatCurve=DataInterface() {propertyValues {Curve {keys [{interpMode=RCIM_Cubic, leaveTangent=0.1821, tangentMode=RCTM_Break, time=0, value=0}, <at 0.3 1>]}}}
      CurveIndex=&Subtract_Float(A=$Emitter.Age, B=$System.EndTime)
     )
    ))
   &ScaleColor(
    Curve Index=$Particles.ScaleRibbonF
    Linear Color Curve=DataInterface() {propertyValues {
      AlphaCurve {keys [<at 0 1>,<at 1 0>]}
      BlueCurve {keys [<at 0 1>,<at 1 1>]}
      GreenCurve {keys [<at 0 0.6667>,<at 1 0.8333>]}
      RedCurve {keys [<at 0 0>,<at 1 0>]}
     }}
    Scale Mode=::NewEnumerator2
   ) {enabled=false}
   &ScaleColor(Curve Index=$Particles.ScaleRibbonF, Scale Mode=::NewEnumerator2) {module=ScaleColor005}
   /SetVariables(Particles.OriginalPosition:NiagaraPosition=$Particles.Position) {enabled=false, isSetParameters=true}
   &self:Into(
    Curl Noise=DataInterface() {propertyValues {Seed=1}}
    Curve for Floats=DataInterface() {propertyValues {Curve {keys [<at 0 1>,<at 1 0>]}}}
    Position:NiagaraPosition=$StackContext.Position
    Quaternion:Quat4f=&MakeQuatFromAxisAngle(Axis=<0 0 1>, AngleInDegrees=&Multiply_Float(A=$Emitter.Age, B=50))
    TorusPos:NiagaraFloat=$Particles.Initial.TorusProgress
    Vector001:Vector3f=<150 0 -150>
    float=0.5
    float001:NiagaraFloat=&Multiply_Float(A=&Multiply_Float(A=&Subtract_Float(A=$Particles.Age, B=0.5), B=2), B=-0.5)
    float002=0.01
    float003=10
   ) {module=Into}
   &ScaleColor(
    Color Value To Scale=$System.Status Color
    Scale Alpha=&FloatFromCurve(
     FloatCurve=DataInterface() {propertyValues {Curve {keys [{interpMode=RCIM_Cubic, tangentMode=RCTM_Break, time=0, value=0}, {interpMode=RCIM_Cubic, tangentMode=RCTM_Break, time=2, value=1}]}}}
     CurveIndex=$Particles.Age
     Scale Curve=&Lerp_Float(
      A=&FloatFromCurve(
       FloatCurve=DataInterface() {propertyValues {Curve {keys [<at 0 0 RCIM_Constant>,<at 0.05 1 RCIM_Constant>,<at 0.95 0>]}}}
       CurveIndex=$Particles.Initial.TorusProgress
      )
      B=1
      Alpha=&MakeCustomFloatFromBool(
       Bool=&SetBoolByFloatComparison(A=$Emitter.Age, B=3)
       True Float=0
       False Float=&FloatFromCurve(FloatCurve=DataInterface() {propertyValues {Curve {keys [<at 1.5 0>,<at 1.8 1>]}}}, CurveIndex=$Particles.Age)
      )
     )
    )
    ScaleA=true
    ScaleRGB=true
   ) {module=ScaleColor001}
   &ScaleColor(
    Color Value To Scale=$System.Status Color
    Scale Alpha=&FloatFromCurve(
     FloatCurve=DataInterface() {propertyValues {Curve {keys [
         {interpMode=RCIM_Cubic, tangentMode=RCTM_Break, time=0, value=0}
         {arriveTangent=2.6295, interpMode=RCIM_Cubic, tangentMode=RCTM_Break, time=1, value=1}
        ]}}}
     CurveIndex=$Particles.IntoProgress
     Scale Curve=&Lerp_Float(
      A=1
      B=&FloatFromCurve(
       FloatCurve=DataInterface() {propertyValues {Curve {keys [
           {interpMode=RCIM_Cubic, leaveTangent=-2.0961, tangentMode=RCTM_Break, time=0, value=1}
           {arriveTangent=0.1831, interpMode=RCIM_Cubic, tangentMode=RCTM_Break, time=1, value=0}
          ]}}}
       CurveIndex=$Particles.Initial.TorusProgress
      )
      Alpha=&FloatFromCurve(
       FloatCurve=DataInterface() {propertyValues {Curve {keys [
           {interpMode=RCIM_Cubic, leaveTangent=-0.2258, tangentMode=RCTM_Break, time=0.5, value=1}
           {arriveTangent=-2.134, interpMode=RCIM_Cubic, tangentMode=RCTM_Break, time=1.5, value=0}
          ]}}}
       CurveIndex=$Particles.Age
      )
     )
    )
    ScaleA=true
    ScaleRGB=true
   ) {module=ScaleColor002}
   /SetVariables(Particles.Velocity=&AddVectorToPosition(
     Position=Linked() {linkedVariable {name=Particles.Velocity, type=Vector3f}}
     Vector=&Lerp_Vector(
      A=<0 0 0>
      B=&Multiply_VectorByFloat(Vector=<0 0 8000>, Float=&RandomRangeFloat(Minimum=1, Maximum=1))
      Alpha=&FloatFromCurve(
       FloatCurve=DataInterface() {propertyValues {Curve {keys [
           {interpMode=RCIM_Cubic, leaveTangent=0.0956, tangentMode=RCTM_Break, time=0, value=0}
           {arriveTangent=2.1575, interpMode=RCIM_Cubic, tangentMode=RCTM_Break, time=0.8, value=1}
           <at 0.9 0>
          ]}}}
       CurveIndex=$Particles.UnCWeight
      )
     )
    )) {enabled=false, isSetParameters=true}
   &LineAttractionForce(
    Attraction Strength:NiagaraFloat=&Lerp_Float(
     A=0
     B=30000
     Alpha=&FloatFromCurve(
      FloatCurve=DataInterface() {propertyValues {Curve {keys [{interpMode=RCIM_Cubic, tangentMode=RCTM_Break, time=0, value=0}, <at 0.2 1>, <at 0.3 0>]}}}
      CurveIndex=$Particles.UnCWeight
     )
    )
    Line End:Vector3f=<0 0 1000>
    Line Start:Vector3f=<0 0 0>
    Use Attraction Falloff=false
   ) {enabled=false}
   /SetVariables(Transient.PhysicsForce:Vector3f=&Multiply_VectorByFloat(
     Vector=$Transient.PhysicsForce
     Float=&ClampFloat(Float=&Divide_Float(A=0.016, B=$Engine.DeltaTime), Max=3)
    )) {enabled=false, isSetParameters=true}
   &SolveForcesAndVelocity(Clamp Velocity=false, Limit Acceleration=false) {enabled=false}
   &ScaleColor(
    Color Value To Scale=$System.Status Color
    Scale Alpha=&FloatFromCurve(
     FloatCurve=DataInterface() {propertyValues {Curve {keys [{interpMode=RCIM_Cubic, leaveTangent=-6.4723, tangentMode=RCTM_Break, time=0, value=1}, <at 0.4 0>]}}}
     CurveIndex=&Subtract_Float(A=$Emitter.Age, B=$System.EndTime)
    )
    ScaleA=true
    ScaleRGB=true
   ) {module=ScaleColor003}
   &ScaleColor(
    Color Value To Scale=$System.Status Color
    Scale Alpha=&FloatFromCurve(FloatCurve=DataInterface() {propertyValues {Curve {keys [<at 0.5 0>,<at 1 1>]}}}, CurveIndex=$User.ResInProgress)
    ScaleA=true
    ScaleRGB=true
   ) {module=ScaleColor004}
   &self:ConstantScreenSize(
    Camera Query=DataInterface() {propertyValues {PlayerControllerIndex=0, bRequireCurrentFrameData=false}}
    Position:NiagaraPosition=$Particles.Position
    Reference Screen Width:NiagaraFloat=$User.ScreenSize
    RibbonWidth:NiagaraFloat=$Particles.RibbonWidth
   ) {module=ConstantScreenSize}
  ]
  rendererClasses [&NiagaraRibbonRendererProperties, &NiagaraSpriteRendererProperties]
  renderers [
   {rendererClass=&NiagaraRibbonRendererProperties}
   {rendererClass=&NiagaraSpriteRendererProperties, rendererIndex=1}
  ]
 }
}
summary {
 systemName=NS_TransisitonIAS
 emitters [
  {emitterName=Cubes, rendererClasses [&NiagaraMeshRendererProperties, &NiagaraSpriteRendererProperties]}
  {emitterName=SoftRibbon1, rendererClasses [&NiagaraRibbonRendererProperties, &NiagaraSpriteRendererProperties]}
  {emitterName=PianoRibbon, rendererClasses [&NiagaraRibbonRendererProperties]}
  {emitterName=DownRibbon, rendererClasses [&NiagaraRibbonRendererProperties, &NiagaraSpriteRendererProperties]}
 ]
 userVariables [
  User.LoadFinished2 {type=NiagaraBool}
  User.ResInProgress {type=NiagaraFloat, defaultValue=1, description=""}
  User.ScreenSize {type=NiagaraFloat, defaultValue=450, description=""}
  User.SKM {
   type=NiagaraDataInterfaceSkeletalMesh
   defaultValue {
    struct=NiagaraExt_VariableValue_DataInterface
    value {dataInterface=&self:NiagaraDataInterfaceSkeletalMesh_0}
   }
  }
  User.Status Color {type=NiagaraInt32}
 ]
}
